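# Sound cue for the looping ambient sound of the Damage Amplifier Creative gadget.
# A single looping wave player feeds the root directly, playing the ambient loop wave at reduced volume under the Creative gadgets sound class.
# The cue overrides attenuation inline to spatialize the loop with its own inner radius, falloff distance and stereo spread instead of a shared attenuation asset.
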SoundCue /Game/Sounds/Creative/Gadgets/DamageAmplifier/DamageAmplifier_AmbientLoop_01_Cue
root=WavePlayer_0
nodes [{id=WavePlayer_0, properties {bLooping=true, soundWaveAssetPtr=&/DamageAmplifier_AmbientLoop_01}}]
properties {
 bOverrideAttenuation=true
 maxDistance=1450
 soundClassObject=&Fort_Gadgets_Creative
 volumeMultiplier=0.2
 attenuationOverrides {falloffDistance=1200, stereoSpread=800, attenuationShapeExtents {x=250}}
}
