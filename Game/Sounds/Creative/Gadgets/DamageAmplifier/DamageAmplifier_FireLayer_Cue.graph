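# Sound cue that plays the fire layer sound for the Damage Amplifier Creative gadget.
# A single wave player feeds the root directly, with no randomization or modulation.
# The cue is routed to the Creative gadgets sound class at reduced volume.
# It overrides attenuation with a logarithmic falloff and stereo spread, and overrides concurrency to cap simultaneous instances.
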SoundCue /Game/Sounds/Creative/Gadgets/DamageAmplifier/DamageAmplifier_FireLayer_Cue
root=WavePlayer_1
nodes [{id=WavePlayer_1, properties {soundWaveAssetPtr=&/DamageAmplifier_FireLayer_01}}]
properties {
 bOverrideAttenuation=true
 bOverrideConcurrency=true
 maxDistance=3800
 soundClassObject=&Fort_Gadgets_Creative
 volumeMultiplier=0.6
 attenuationOverrides {distanceAlgorithm=Logarithmic, falloffDistance=3000, stereoSpread=800, attenuationShapeExtents {x=800}}
 concurrencyOverrides {maxCount=3}
}
