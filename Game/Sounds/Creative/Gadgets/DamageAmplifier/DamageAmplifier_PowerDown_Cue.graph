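# Sound cue for the power-down sound of the Creative Damage Amplifier gadget.
# A single wave player feeds the root directly, with no randomization or modulation.
# The cue routes to the Creative gadgets sound class at reduced volume and overrides attenuation with its own falloff, spherical extent and stereo spread for positional playback.
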SoundCue /Game/Sounds/Creative/Gadgets/DamageAmplifier/DamageAmplifier_PowerDown_Cue
root=WavePlayer_1
nodes [{id=WavePlayer_1, properties {soundWaveAssetPtr=&/DamageAmplifier_PowerDown_01}}]
properties {
 bOverrideAttenuation=true
 maxDistance=3800
 soundClassObject=&Fort_Gadgets_Creative
 volumeMultiplier=0.4
 attenuationOverrides {falloffDistance=3000, stereoSpread=800, attenuationShapeExtents {x=800}}
}
