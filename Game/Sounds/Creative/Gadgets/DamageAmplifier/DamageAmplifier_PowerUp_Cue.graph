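# Sound cue for the power-up sound of the Creative Damage Amplifier gadget.
# A single wave player feeds DamageAmplifier_PowerUp_01 straight to the output, with no randomization or modulation.
# It is routed to the Creative gadgets sound class and overrides attenuation for its own falloff, stereo spread and shape extent.
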SoundCue /Game/Sounds/Creative/Gadgets/DamageAmplifier/DamageAmplifier_PowerUp_Cue
root=WavePlayer_0
nodes [{id=WavePlayer_0, properties {soundWaveAssetPtr=&/DamageAmplifier_PowerUp_01}}]
properties {
 bOverrideAttenuation=true
 maxDistance=3800
 soundClassObject=&Fort_Gadgets_Creative
 attenuationOverrides {falloffDistance=3000, stereoSpread=800, attenuationShapeExtents {x=800}}
}
