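# Looping sound cue for the powered state of the Creative Damage Amplifier gadget.
# A single looping wave player feeds the output directly, with no randomization or mixing.
# It plays at reduced volume in the Creative gadgets sound class and overrides attenuation with its own falloff distance and stereo spread.
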SoundCue /Game/Sounds/Creative/Gadgets/DamageAmplifier/DamageAmplifier_PoweredLoop_Cue
root=WavePlayer_1
nodes [{id=WavePlayer_1, properties {bLooping=true, soundWaveAssetPtr=&/DamageAmplifier_PoweredLoop_01}}]
properties {
 bOverrideAttenuation=true
 maxDistance=1800
 soundClassObject=&Fort_Gadgets_Creative
 volumeMultiplier=0.3
 attenuationOverrides {falloffDistance=1400, stereoSpread=800}
}
